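# Parent translucent unlit material for a MedMist smoke volume, with static switches for translucent and shadow variants.
# MF_RayMarchSphere samples of 3D noise, trimmed to two for normals and mask, feed sun-based shading tinted by time of day.
# Opacity combines fresnel, height, depth fade and custom-data timestamp transitions, with dithered cloud-texture shadows.
# WPO displaces shadows by Voronoi noise or pushes vertices camera-relative, and drops geometry far down to avoid overdraw.
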Material /MedCloud/Assets/MedMistSmokeVolume/M_MedMist_Parent
graph {
 properties {
  bAutomaticallySetUsageInEditor=false
  bTangentSpaceNormal=false
  blendMode=BLEND_Translucent
  shadingModel=MSM_Unlit
  translucencyPass=MTP_BeforeDOF
  usedShadingModels=Unlit
 }
 nodes [
  n1 = ScalarParameter() {defaultValue=0.75, parameterName=Scale}
  n2 = TextureObjectParameter() {parameterName=VolumeTexture, samplerType=SAMPLERTYPE_LinearColor, texture=&/T_3DNoise_N}
  n3 = Multiply(
   A=Multiply(
    A=VectorParameter() {parameterName=Param_1, defaultValue=<rgba 1 1 -1.5 1>}.RGB
    B=ScalarParameter() {defaultValue=2.5, parameterName=OverallSpeed}
   )
   B=Multiply(A=Time()) {constB=0.025}
  )
  n4 = MaterialFunctionCall!(Scale=@n1, VolumeTexture=@n2, In, ObjectPositionOffset=@n3) {materialFunction=&/MF_RayMarchSphere}
  n5 = Multiply(A=Multiply(
    A=ScalarParameter() {defaultValue=8, parameterName=PlaneDistance}
    B=MaterialFunctionCall() {materialFunction=&ObjectScale}.Scale XYZ
   )) {constB=2}
  n6 = MaterialFunctionCall(Scale=@n1, VolumeTexture=@n2, In=@n5, ObjectPositionOffset=@n3) {materialFunction=&MF_RayMarchSphere}
  n7 = Add(A=@n4.Mask, B=@n6.Mask)
  n8 = Multiply(A=@n5) {constB=2}
  n9 = MaterialFunctionCall(Scale=@n1, VolumeTexture=@n2, In=@n8, ObjectPositionOffset=@n3) {materialFunction=&MF_RayMarchSphere}
  n10 = MaterialFunctionCall(Scale=@n1, VolumeTexture=@n2, In=Multiply(A=@n8) {constB=2}, ObjectPositionOffset=@n3) {materialFunction=&MF_RayMarchSphere}
  Add(A=Add(A=@n7, B=@n9.Mask), B=@n10.Mask)
  n11 = ScalarParameter() {defaultValue=12, parameterName=PlaneFresnelFade}
  n12 = Multiply(A=OneMinus(None=Fresnel(ExponentIn=@n11) {baseReflectFraction=0, exponent=1})) {constB=4}
  n13 = Subtract(A=@n12)
  n14 = Add(A=Add(A=@n4.RGB, B=@n4.RGB), B=Multiply(A=Add(A=@n6.RGB, B=@n6.RGB), B=Saturate(None=@n13)))
  n15 = Add(A=@n9.RGB, B=@n9.RGB)
  n16 = Saturate(None=Subtract(A=@n12) {constB=2})
  n17 = Subtract(A=@n12) {constB=3}
  Add(A=Add(A=@n14, B=Multiply(A=@n15, B=@n16)), B=Multiply(A=Add(A=@n10.RGB, B=@n10.RGB), B=Saturate(None=@n17)))
  Constant()
  LinearInterpolate(A=NamedRerouteUsage(), Alpha=ScalarParameter() {parameterName=IsOpaque, primitiveDataIndex=7})
  NamedRerouteDeclaration(None=Multiply(A=Divide(A=@n14) {desc=Using 2 samples instead}, B=ScalarParameter() {defaultValue=1, parameterName=NormalIntensity})) {name=Normals, nodeColor=<rgba 0.0078 1 0 1>}
  NamedRerouteDeclaration(None=Divide(A=@n7) {desc=Using 2 samples instead}) {name=MaskSum, nodeColor=<rgba 1 0.6328 0 1>}
  n18 = Abs(None=ComponentMask(None=VertexNormalWS()) {b=true})
  n19 = TextureCoordinate()
  n20 = ScalarParameter() {defaultValue=4, parameterName=TextureSize}
  n21 = TextureSample(UVs=Panner(Coordinate=Multiply(A=@n19, B=@n20)) {speedY=0.05}) {texture=/Game/Effects/Fort_Effects/Textures/T_PinkSpike_Voronoi2D_Height}
  n22 = Multiply(
   A=SquareRoot(None=OneMinus(None=@n18))
   B=Subtract(A=NamedRerouteDeclaration(None=@n21.G) {name=2DNoise, nodeColor=<rgba 0.1953 0 1 1>}) {constB=0.5}
  )
  n23 = Multiply(B=Divide(A=MaterialFunctionCall() {materialFunction=&ObjectScale}.Scale XYZ)) {constA=100, constB=5}
  n24 = Multiply(A=Multiply(A=ScalarParameter() {defaultValue=0.2, parameterName=WPOIntensity, primitiveDataIndex=8}), B=@n23)
  n25 = VertexNormalWS()
  n26 = ScalarParameter() {defaultValue=0.051, parameterName=WPONormalIntensity, primitiveDataIndex=8}
  n27 = Add(A=Multiply(A=Multiply(A=@n22, B=@n24), B=@n25), B=Multiply(A=@n25, B=Multiply(A=@n23, B=@n26)))
  n28 = WorldPosition() {worldPositionShaderOffset=WPT_CameraRelative}
  n29 = ScalarParameter() {defaultValue=0.35, parameterName=PushOpaque}
  n30 = VertexColor()
  n31 = ScalarParameter() {parameterName=PushTranslucent}
  n32 = Add(A=Multiply(A=Multiply(A=@n28.XYZ, B=@n29) {constB=10}, B=@n30.R), B=Multiply(A=Multiply(A=@n28.XYZ, B=@n31), B=OneMinus(None=@n30.R)))
  n33 = OneMinus(None=Ceil(None=NamedRerouteUsage()))
  n34 = Constant3Vector() {constant=<rgba 0 0 -500000 0>}
  NamedRerouteDeclaration(None=Add(A=StaticSwitchParameter(True=@n27, False=@n32) {parameterName=IsShadow}, B=Multiply(A=@n33, B=@n34) {desc=Move this below to avoid overdraw}) {constB=0}) {name=WPO, nodeColor=<rgba 0 0.1797 1 1>}
  NamedRerouteDeclaration(None=ScalarParameter() {bUseCustomPrimitiveData=true, parameterName=InvertTimestamp, primitiveDataIndex=1}) {name=InvertTimestamp, nodeColor=<rgba 1 0.9375 0 1>}
  n35 = ConstantBiasScale(None=DotProduct(A=NamedRerouteUsage(), B=SkyAtmosphereLightDirection()))
  n36 = VectorParameter() {
   bUseCustomPrimitiveData=true
   parameterName=Color Multiplier
   primitiveDataIndex=3
   channelNames {a="", b="0,479338", g="1,0", r="0,001823"}
   defaultValue=<rgba 0.0467 0.6406 0.3078 1>
  }
  n37 = Multiply(
   A=LinearInterpolate(
    A=ScalarParameter() {defaultValue=0.15, parameterName=SmokeMin}
    B=ScalarParameter() {defaultValue=1.2, parameterName=SmokeMax}
    Alpha=Multiply(A=@n35, B=@n35)
   ) {constA=0.2}
   B=@n36
  )
  n38 = Multiply(A=@n37, B=ScalarParameter() {bUseCustomPrimitiveData=true, defaultValue=0.75, parameterName=ToD_NightMultiplier, primitiveDataIndex=2}) {constB=0.5}
  n39 = VectorParameter() {bUseCustomPrimitiveData=true, parameterName=ToD_EveningMultiplier, primitiveDataIndex=8, defaultValue=<rgba 1 0.4101 0.3886 1>}
  n40 = VectorParameter() {bUseCustomPrimitiveData=true, parameterName=ToD_NightMultiplier, primitiveDataIndex=12, defaultValue=<rgba 0.3529 0.538 1 1>}
  n41 = Constant() {r=1}
  n42 = Constant() {r=1.1}
  n43 = Multiply(
   A=MaterialFunctionCall(Color_Morning=@n37, Color_Day=@n37, Color_Evening=Multiply(A=@n38, B=@n39.RGB), Color_Night=Multiply(A=@n38, B=@n40.RGB)) {materialFunction=&Color_Time_of_Day}.Result
   B=QualitySwitch(Default=@n41, Low=@n42, High=@n41, Medium=@n42, Epic=@n41)
  ) {constB=1.2}
  n44 = EyeAdaptationInverse()
  NamedRerouteDeclaration(None=StaticSwitchParameter(True=Constant(), False=Multiply(A=@n43, B=@n44.EyeAdaptationInverse)) {parameterName=IsShadow}) {name=Emissive, nodeColor=<rgba 0.3828 1 0 1>}
  n45 = Distance(A=TransformPosition(None=Constant3Vector()) {transformType=TRANSFORMPOSSOURCE_TranslatedWorld}, B=Constant3Vector())
  n46 = Multiply(A=Subtract(A=Subtract(A=@n45) {constB=500}, B=ScalarParameter() {defaultValue=375, parameterName=Offset, primitiveDataIndex=18})) {constB=-1}
  n47 = ScalarParameter() {defaultValue=32, parameterName=Fade, primitiveDataIndex=18}
  n48 = OneMinus(None=Saturate(None=Divide(A=@n46, B=@n47)))
  NamedRerouteDeclaration(None=VertexInterpolator(VS=NamedRerouteDeclaration(None=@n48) {name=DistanceFadeMask, nodeColor=<rgba 1 0.0938 0 1>}).PS) {name=VertexDistanceFadeMask, nodeColor=<rgba 1 0.0938 0 1>}
  n49 = Constant()
  n50 = Subtract(
   A=WorldPosition().XYZ
   B=MaterialFunctionCall() {materialFunction=&ObjectPivotPoint}.Mesh Particle Pivot Location
  )
  n51 = ComponentMask(None=Transform(None=Divide(A=@n50) {constB=1000}) {transformSourceType=TRANSFORMSOURCE_World, transformType=TRANSFORM_View}) {g=true, r=true}
  n52 = Multiply(A=Time(), B=Constant2Vector() {g=0.066, r=0.073}) {constB=0.02}
  n53 = TextureSample(UVs=VertexInterpolator(VS=Add(A=@n51, B=@n52)).PS) {samplerType=SAMPLERTYPE_LinearGrayscale, texture=/Game/Environments/Helios/TechArt/LowAltitudeClouds/T_LowAltitudeClouds_A}
  n54 = SkyAtmosphereLightDirection()
  n55 = VertexNormalWS()
  n56 = Multiply(A=Power(Base=DotProduct(A=@n54, B=@n55))) {constB=2}
  n57 = Divide(A=LocalPosition().Z) {constB=50}
  n58 = Multiply(A=Multiply(A=Saturate(None=@n56), B=Saturate(None=@n57)), B=NamedRerouteUsage())
  n59 = MaterialFunctionCall(Alpha Threshold=Multiply(A=Add(A=@n53.R, B=@n58), B=@n58)) {materialFunction=&DitherTemporalAA}
  n60 = NamedRerouteUsage()
  n61 = Fresnel(ExponentIn=ScalarParameter() {defaultValue=5, parameterName=FresnelFade, primitiveDataIndex=1}) {baseReflectFraction=0}
  n62 = MaterialFunctionCall() {materialFunction=&BoundingBoxBased_0-1_UVW}
  n63 = NamedRerouteUsage()
  n64 = LinearInterpolate(A=@n62.B, B=OneMinus(None=@n62.B), Alpha=OneMinus(None=Floor(None=@n63)))
  n65 = OneMinus(None=NamedRerouteUsage())
  n66 = Power(Base=Subtract(A=Add(A=@n64), B=LinearInterpolate(Alpha=@n65) {constB=1.75}), Exp=ScalarParameter() {defaultValue=0.05, parameterName=Exp})
  n67 = SmoothStep(Min=NamedRerouteUsage(), Value=Multiply(A=OneMinus(None=@n61), B=Saturate(None=@n66)))
  n68 = ScalarParameter() {defaultValue=1, parameterName=OpacityMultiplier}
  n69 = Multiply(A=Multiply(A=LinearInterpolate(Alpha=@n60) {constA=0.5}, B=Multiply(A=@n67, B=@n68)), B=NamedRerouteUsage())
  n70 = ScalarParameter() {defaultValue=124, parameterName=DepthFade}
  n71 = Multiply(A=@n69, B=NamedRerouteDeclaration(None=DepthFade(FadeDistance=@n70)) {name=DepthFade, nodeColor=<rgba 1 0.5391 0 1>})
  NamedRerouteDeclaration(None=StaticSwitchParameter(True=ShadowReplace(Default=@n49, Shadow=@n59.Result), False=StaticSwitchParameter(True=@n71, False=@n69) {parameterName=IsTranslucent}) {parameterName=IsShadow}) {name=Opacity, nodeColor=<rgba 0 0.6484 1 1>}
  n72 = Subtract(A=Time(), B=ScalarParameter() {bUseCustomPrimitiveData=true, parameterName=InTimestamp})
  n73 = ScalarParameter() {bUseCustomPrimitiveData=true, defaultValue=1.5, desc=InLength, parameterName=TransitionLength, primitiveDataIndex=16}
  n74 = Multiply(A=@n73) {constB=-1}
  n75 = NamedRerouteUsage()
  n76 = LinearInterpolate(
   A=@n72
   B=Subtract(A=@n72, B=LinearInterpolate(A=@n73, B=@n74, Alpha=@n75))
   Alpha=ScalarParameter() {parameterName=IsOpaque, primitiveDataIndex=7}
  )
  n77 = NamedRerouteDeclaration(None=Clamp(None=Divide(A=@n76, B=@n73) {desc=InLength})) {name=InTimestamp, nodeColor=<rgba 1 0 0.0469 1>}
  n78 = LinearInterpolate(
   A=LinearInterpolate(A=@n77, B=OneMinus(None=@n77), Alpha=NamedRerouteUsage())
   Alpha=ScalarParameter() {bUseCustomPrimitiveData=true, parameterName=SmokeDebug, primitiveDataIndex=7}
  )
  NamedRerouteDeclaration(None=VertexInterpolator(VS=Saturate(None=@n78)).PS) {name=Timestamp, nodeColor=<rgba 0 0.6953 1 1>}
  NamedRerouteUsage()
  n79 = Multiply(A=Time()) {constB=0.1}
  Saturate(None=ConstantBiasScale(None=Sine(None=@n79)))
  n80 = NamedRerouteUsage()
  ShadowReplace(
   Default=MaterialFunctionCall(Alpha Threshold=@n80) {materialFunction=&DitherTemporalAA}.Result
   Shadow=@n80
  )
 ]
}
