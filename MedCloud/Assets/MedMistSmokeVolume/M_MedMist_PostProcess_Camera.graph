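# Post-process material for the MedMist smoke volume that fogs the screen when the camera is near ActorLocation, blending over scene color before bloom.
# Animated 3D noise from a 2D slice atlas, sampled by view direction and camera position, supplies opacity and sun-direction lighting.
# Color follows time of day and compensates for eye adaptation.
# Distance fade limits the effect, alongside a timestamp-driven transition with invert and debug controls.
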Material /MedCloud/Assets/MedMistSmokeVolume/M_MedMist_PostProcess_Camera
graph {
 properties {
  bDisableDepthTest=true
  bUseTranslucencyVertexFog=false
  bUsedWithNiagaraMeshParticles=true
  bUsedWithNiagaraSprites=true
  blendMode=BLEND_Translucent
  blendableLocation=BL_SceneColorBeforeBloom
  materialDomain=MD_PostProcess
  pixelDepthOffsetMode=PDOM_Legacy
  shadingModel=MSM_Unlit
  usedShadingModels=Unlit
 }
 nodes [
  n1 = Multiply(A=Multiply(A=CameraVectorWS()) {constB=-1}, B=ScalarParameter() {defaultValue=0.75, parameterName=PP_NoiseScale}) {constB=0.01}
  n2 = Time()
  n3 = Multiply(
   A=Multiply(A=Constant3Vector() {constant=<rgba 1 1 -1 1>}, B=ScalarParameter() {defaultValue=1.5, parameterName=PP_NoiseSpeed})
   B=Fmod(A=Multiply(A=@n2) {constB=0.05}, B=Constant() {r=1000})
  )
  Add(A=@n1, B=@n3)
  n4 = Multiply(A=CameraPositionWS()) {constB=-0.0007}
  n5 = MaterialFunctionCall!(
   UVs=Add(A=Subtract(A=@n1, B=@n4), B=@n3)
   Slices Texture=TextureObject() {samplerType=SAMPLERTYPE_LinearColor, texture=&/T_3DNoise_N}
   Atlas Size=Constant() {r=8}
   Atlas Padding=ScalarParameter() {defaultValue=0.9404, parameterName=AtlasPadding, primitiveDataIndex=8}
  ) {materialFunction=/Game/Environments/Helios/TechArt/LowAltitudeClouds/MF_3DVolumeFrom2DSlices}
  n6 = Add(A=@n5.RGB) {constB=-0.5}
  n7 = ConstantBiasScale(None=DotProduct(A=Multiply(A=@n6) {constB=2}, B=SkyAtmosphereLightDirection()))
  n8 = Multiply(
   A=LinearInterpolate(Alpha=Multiply(A=@n7, B=@n7)) {constA=0.25}
   B=VectorParameter() {parameterName=Color Multiplier, primitiveDataIndex=3, defaultValue=<rgba 0 0.7812 0.3482 1>}.RGB
  )
  n9 = Multiply(A=@n8, B=ScalarParameter() {defaultValue=1, parameterName=ToD_NightMultiplier, primitiveDataIndex=2})
  n10 = Multiply(A=@n9, B=VectorParameter() {parameterName=ToD_EveningMultiplier, primitiveDataIndex=8, defaultValue=<rgba 1 0.4101 0.3886 1>}.RGB)
  n11 = Multiply(A=@n9, B=VectorParameter() {parameterName=ToD_NightMultiplier, primitiveDataIndex=12, defaultValue=<rgba 0.3529 0.538 1 1>}.RGB)
  n12 = NamedRerouteUsage()
  n13 = LinearInterpolate(
   A=ScalarParameter() {defaultValue=0.5, parameterName=MinOpacity, sliderMax=1}
   B=ScalarParameter() {defaultValue=0.9, parameterName=MaxOpacity, sliderMax=1}
   Alpha=Multiply(A=@n12, B=@n12)
  ) {constA=0.5}
  n14 = CameraPositionWS()
  n15 = VectorParameter() {parameterName=ActorLocation, defaultValue=<rgba 0 0 0 1>}
  n16 = Subtract(A=Subtract(A=Distance(A=@n14, B=@n15.RGB)) {constB=500}, B=ScalarParameter() {defaultValue=400, parameterName=Offset, primitiveDataIndex=18})
  n17 = Saturate(None=Divide(A=Multiply(A=@n16) {constB=-1}, B=ScalarParameter() {defaultValue=32, parameterName=Fade, primitiveDataIndex=18}))
  LinearInterpolate(
   A=ComponentMask(None=SceneTexture() {bFiltered=true, sceneTextureId=PPI_PostProcessInput0}.Color) {b=true, g=true, r=true}
   B=Multiply(
    A=MaterialFunctionCall(Color_Morning=@n8, Color_Day=@n8, Color_Evening=@n10, Color_Night=@n11) {materialFunction=&Color_Time_of_Day}.Result
    B=EyeAdaptationInverse().EyeAdaptationInverse
   )
   Alpha=Multiply(A=Multiply(A=@n13, B=@n17), B=NamedRerouteUsage())
  )
  NamedRerouteDeclaration(None=@n5.A) {name=Mask, nodeColor=<rgba 0 0.0156 1 1>}
  n18 = Subtract(A=Time(), B=ScalarParameter() {parameterName=InTimestamp})
  n19 = ScalarParameter() {defaultValue=1.5, desc=InLength, parameterName=TransitionLength, primitiveDataIndex=16}
  n20 = NamedRerouteDeclaration(None=Clamp(None=Divide(A=@n18, B=@n19) {desc=InLength})) {name=InTimestamp, nodeColor=<rgba 1 0 0.0469 1>}
  n21 = LinearInterpolate(A=@n20, B=OneMinus(None=@n20), Alpha=ScalarParameter() {parameterName=InvertTimestamp, primitiveDataIndex=1})
  n22 = ScalarParameter() {parameterName=SmokeDebug, primitiveDataIndex=7}
  NamedRerouteDeclaration(None=VertexInterpolator(VS=LinearInterpolate(A=@n21, Alpha=@n22)).PS) {name=Timestamp, nodeColor=<rgba 0 0.6953 1 1>}
  Reroute(None=Multiply(A=Saturate(None=@n16), B=Constant3Vector() {constant=<rgba 0 0 -500000 0>}) {desc=Move this below to avoid overdraw})
  TransformPosition(None=Constant3Vector()) {transformType=TRANSFORMPOSSOURCE_TranslatedWorld}
  VertexInterpolator(VS=@n17)
 ]
}
